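# Physics asset for the face accessory part of the female medium VileSting Crude cosmetic body, set up for dynamic hair secondary motion.
# A large head capsule anchors small base capsules for four left and four right strands plus front and back strands, with angular motion locked to the head.
# Each strand adds one capsule hinged to its base with narrow swing and tight twist limits, allowing restrained sway.
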PhysicsAsset /BRCosmetics/Characters/Player/Female/Medium/Bodies/F_MED_VileSting_Crude/Meshes/Parts/F_MED_VileSting_Crude_FaceAcc_Physics
bodies {
 dyn_hair_1_base_l [{shapeType=Capsule, length=0, radius=1, shapeName=dyn_hair_1_base_l_capsule, rotation=<rot 0 -90 0>}]
 dyn_hair_1_base_r [{shapeType=Capsule, length=0, radius=1, shapeName=dyn_hair_1_base_r_capsule, rotation=<rot 0 -90 0>}]
 dyn_hair_1_l [{
   shapeType=Capsule
   length=3.7233
   radius=1.2955
   shapeName=dyn_hair_1_l_capsule
   center=<3.0105 0.436 0.5363>
   rotation=<rot 3.5231 104.3518 -99.7377>
  }]
 dyn_hair_1_r [{
   shapeType=Capsule
   length=4.903
   radius=1.1927
   shapeName=dyn_hair_1_r_capsule
   center=<3.741 0.8199 -0.4335>
   rotation=<rot 12.4516 110.6196 -81.3489>
  }]
 dyn_hair_2_base_l [{shapeType=Capsule, length=0, radius=1, shapeName=dyn_hair_2_base_l_capsule, rotation=<rot 0 -90 0>}]
 dyn_hair_2_base_r [{shapeType=Capsule, length=0, radius=1, shapeName=dyn_hair_2_base_r_capsule, rotation=<rot 0 -90 0>}]
 dyn_hair_2_l [{
   shapeType=Capsule
   length=3.755
   radius=1.005
   shapeName=dyn_hair_2_l_capsule
   center=<3.4245 0.1588 0.1765>
   rotation=<rot -3.9859 -84.2055 -79.9282>
  }]
 dyn_hair_2_r [{
   shapeType=Capsule
   length=2.755
   radius=1.005
   shapeName=dyn_hair_2_r_capsule
   center=<3.0409 -0.431 -0.8098>
   rotation=<rot 5.748 2.6835 -72.2384>
  }]
 dyn_hair_3_base_l [{shapeType=Capsule, length=0, radius=1, shapeName=dyn_hair_3_base_l_capsule, rotation=<rot 0 -90 0>}]
 dyn_hair_3_base_r [{shapeType=Capsule, length=0, radius=1, shapeName=dyn_hair_3_base_r_capsule, rotation=<rot 0 -90 0>}]
 dyn_hair_3_l [{
   shapeType=Capsule
   length=3.606
   radius=1.1485
   shapeName=dyn_hair_3_l_capsule
   center=<2.7725 -0.1234 0.7424>
   rotation=<rot 12.8657 -176.243 -89.9999>
  }]
 dyn_hair_3_r [{
   shapeType=Capsule
   length=2.4874
   radius=1.1301
   shapeName=dyn_hair_3_r_capsule
   center=<2.7537 0.0204 -1.5975>
   rotation=<rot 1.4574 -90.5811 -98.296>
  }]
 dyn_hair_4_base_l [{shapeType=Capsule, length=0, radius=1, shapeName=dyn_hair_4_base_l_capsule, rotation=<rot 0 -90 0>}]
 dyn_hair_4_base_r [{shapeType=Capsule, length=0, radius=1, shapeName=dyn_hair_4_base_r_capsule, rotation=<rot 0 -90 0>}]
 dyn_hair_4_l [{
   shapeType=Capsule
   length=3.3878
   radius=1.2284
   shapeName=dyn_hair_4_l_capsule
   center=<2.7089 -0.0466 0.3572>
   rotation=<rot -2.2987 -177.6909 -64.053>
  }]
 dyn_hair_4_r [{
   shapeType=Capsule
   length=2.9586
   radius=0.9366
   shapeName=dyn_hair_4_r_capsule
   center=<3.8219 -0.5695 -0.0011>
   rotation=<rot -0.1626 105.025 -88.1406>
  }]
 dyn_hair_b [{
   shapeType=Capsule
   length=6.1398
   radius=1.6049
   shapeName=dyn_hair_b_capsule
   center=<3.5683 0.2845 1.4301>
   rotation=<rot -4.6111 2.7334 -89.7312>
  }]
 dyn_hair_base_b [{shapeType=Capsule, length=0, radius=1, shapeName=dyn_hair_base_b_capsule, rotation=<rot 0 -90 0>}]
 dyn_hair_front [{
   shapeType=Capsule
   length=4.6471
   radius=1.1029
   shapeName=dyn_hair_front_capsule
   center=<0.1739 0.4421 -3.4075>
   rotation=<rot 0 86.9293 -89.7203>
  }]
 dyn_hair_front_base [{shapeType=Capsule, length=0, radius=1, shapeName=dyn_hair_front_base_capsule, rotation=<rot 0 -90 0>}]
 head [{
   shapeType=Capsule
   length=1.25
   radius=8.6676
   shapeName=head_capsule
   center=<6.1828 -0.8262 0.1007>
   rotation=<rot -0.3185 -83.5982 -91.1886>
  }]
}
constraints [
 {bone1Name=dyn_hair_1_l, bone2Name=dyn_hair_1_base_l, swing1LimitDegrees=40, swing2LimitDegrees=20, twistLimitDegrees=5}
 {bone1Name=dyn_hair_1_r, bone2Name=dyn_hair_1_base_r, swing1LimitDegrees=40, swing2LimitDegrees=20, twistLimitDegrees=5}
 {bone1Name=dyn_hair_2_l, bone2Name=dyn_hair_2_base_l, swing1LimitDegrees=10, swing2LimitDegrees=30, twistLimitDegrees=5}
 {bone1Name=dyn_hair_2_r, bone2Name=dyn_hair_2_base_r, swing1LimitDegrees=15, swing2LimitDegrees=30, twistLimitDegrees=5}
 {bone1Name=dyn_hair_3_l, bone2Name=dyn_hair_3_base_l, swing1LimitDegrees=15, swing2LimitDegrees=30, twistLimitDegrees=5}
 {bone1Name=dyn_hair_3_r, bone2Name=dyn_hair_3_base_r, swing1LimitDegrees=15, swing2LimitDegrees=30, twistLimitDegrees=5}
 {bone1Name=dyn_hair_4_l, bone2Name=dyn_hair_4_base_l, swing1LimitDegrees=15, swing2LimitDegrees=30, twistLimitDegrees=5}
 {bone1Name=dyn_hair_4_r, bone2Name=dyn_hair_4_base_r, swing1LimitDegrees=15, swing2LimitDegrees=30, twistLimitDegrees=5}
 {bone1Name=dyn_hair_front, bone2Name=dyn_hair_front_base, swing1LimitDegrees=15, swing2LimitDegrees=15, twistLimitDegrees=5}
 {
  bone1Name=dyn_hair_front_base
  bone2Name=head
  swing1LimitDegrees=45
  swing1Motion=Locked
  swing2LimitDegrees=45
  swing2Motion=Locked
  twistLimitDegrees=45
  twistMotion=Locked
 }
 {
  bone1Name=dyn_hair_1_base_r
  bone2Name=head
  swing1LimitDegrees=45
  swing1Motion=Locked
  swing2LimitDegrees=45
  swing2Motion=Locked
  twistLimitDegrees=45
  twistMotion=Locked
 }
 {
  bone1Name=dyn_hair_2_base_r
  bone2Name=head
  swing1LimitDegrees=45
  swing1Motion=Locked
  swing2LimitDegrees=45
  swing2Motion=Locked
  twistLimitDegrees=45
  twistMotion=Locked
 }
 {
  bone1Name=dyn_hair_3_base_r
  bone2Name=head
  swing1LimitDegrees=45
  swing1Motion=Locked
  swing2LimitDegrees=45
  swing2Motion=Locked
  twistLimitDegrees=45
  twistMotion=Locked
 }
 {
  bone1Name=dyn_hair_4_base_r
  bone2Name=head
  swing1LimitDegrees=45
  swing1Motion=Locked
  swing2LimitDegrees=45
  swing2Motion=Locked
  twistLimitDegrees=45
  twistMotion=Locked
 }
 {
  bone1Name=dyn_hair_4_base_l
  bone2Name=head
  swing1LimitDegrees=45
  swing1Motion=Locked
  swing2LimitDegrees=45
  swing2Motion=Locked
  twistLimitDegrees=45
  twistMotion=Locked
 }
 {
  bone1Name=dyn_hair_3_base_l
  bone2Name=head
  swing1LimitDegrees=45
  swing1Motion=Locked
  swing2LimitDegrees=45
  swing2Motion=Locked
  twistLimitDegrees=45
  twistMotion=Locked
 }
 {
  bone1Name=dyn_hair_2_base_l
  bone2Name=head
  swing1LimitDegrees=45
  swing1Motion=Locked
  swing2LimitDegrees=45
  swing2Motion=Locked
  twistLimitDegrees=45
  twistMotion=Locked
 }
 {
  bone1Name=dyn_hair_1_base_l
  bone2Name=head
  swing1LimitDegrees=45
  swing1Motion=Locked
  swing2LimitDegrees=45
  swing2Motion=Locked
  twistLimitDegrees=45
  twistMotion=Locked
 }
 {
  bone1Name=dyn_hair_base_b
  bone2Name=head
  swing1LimitDegrees=45
  swing1Motion=Locked
  swing2LimitDegrees=45
  swing2Motion=Locked
  twistLimitDegrees=45
  twistMotion=Locked
 }
 {bone1Name=dyn_hair_b, bone2Name=dyn_hair_base_b, swing1LimitDegrees=15, swing2LimitDegrees=15, twistLimitDegrees=5}
]
